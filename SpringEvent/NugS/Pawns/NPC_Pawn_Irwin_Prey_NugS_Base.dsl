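; Base pawn for the Spring Event chicken prey NPC (NugS), handling random variants and phasing in and out.
; Chickens tagged Athena.AI.NugS.SFE (egg spawns) begin phased out with collision, interaction and animation suppressed, with a delayed unhide fallback.
; A PhasedIn RepNotify triggers phase-in, restoring saved collision profiles and removing the phased-out tag.
; ApplyVariation picks a hotfixable weighted Slurp or Hop variant, setting material, color value, quest tag and marker icon.
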(event EventBeginPlay
  ;[0.578 0.524 1] Begin Play
  ; Phasing out the chicken if it spawned from an egg
  ; Randomly choosing its variant if spawned naturally
  (|CallParentFunction)
  (if (and (GameplayTags|HasMatchingGameplayTag "(TagName=\"Athena.AI.NugS.SFE\")") (not (Variables|Default|GetPhasedIn)))
    (CallFunction|PhaseOutNugS)
    (Utilities|FlowControl|Delay 3.0) ; Fall back to make sure actor gets unhidden
    (Rendering|SetActorHiddenInGame false)
    (else
      (Rendering|SetActorHiddenInGame false))))

(event Custom|PhaseOutNugS
  ;[1 0.5417 0.5417] PhaseOutNugS
  ; Phasing out the chicken, temporarily disabling its collision, visuals, animation and behavior
  (Variables|Default|SetCapsuleCollisionProfile (Collision|GetCollisionProfileName (Variables|Character|GetCapsuleComponent)))
  (Variables|Default|SetMeshCollisionProfile (Collision|GetCollisionProfileName (Variables|Character|GetMesh)))
  (CallFunction|SetNugSCollisionEnabled false)
  (Networking|SwitchHasAuthority
    (:Authority
      (Ability|GameplayEffect|AddLooseGameplayTags self (GameplayTags|MakeGameplayTagContainerfromTag "(TagName=\"SpringChicken.PhasedOut\")") true))
    (:Remote
      (Variables|/Script/Engine.SkeletalMeshComponent|SetbPauseAnims true (Variables|Character|GetMesh)))))

(event Custom|PhaseInNugS
  ;[0.6875 1 0.6402] PhaseInNugS
  ; Reenabling everything that was disabled when the chicken was phased out
  (CallFunction|SetNugSCollisionEnabled true)
  (Networking|SwitchHasAuthority
    (:Authority
      _)
    (:Remote
      (Variables|/Script/Engine.SkeletalMeshComponent|SetbPauseAnims false (Variables|Character|GetMesh))
      (Rendering|SetActorHiddenInGame false)))
  (Ability|GameplayEffect|RemoveLooseGameplayTags self (GameplayTags|MakeGameplayTagContainerfromTag "(TagName=\"SpringChicken.PhasedOut\")") true))

(fn ApplyVariation
  (bind _selectedvariant (LocalVariables|GetSelectedVariant))
  (bind _selectedvariant_2 (LocalVariables|GetSelectedVariant))
  (bind (_markerdisplay_displayname _markerdisplay_icon _markerdisplay_customindicatorclass _markerdisplay_customdetailsclass _markerdisplay_primarycolor _markerdisplay_sound _markerdisplay_battachsound _markerdisplay_screenclamping) (Utilities|Struct|BreakMarkedActorDisplayInfo (Variables|GetMarkerDisplay_DisplayName)))
  (|CallParentFunction)
  (LocalVariables|SetSelectedVariant (Utilities|Select :Index (Math|Fortnite|GetRandomWeightedIndex (Utilities|Array|MakeArray (ScalableFloat|GetValueatLevel (Variables|Default|GetVariant1WeightHotfix) 1.0) (ScalableFloat|GetValueatLevel (Variables|Default|GetVariant2WeightHotfix) 1.0))) :Option%200 "NewEnumerator0" :Option%201 "NewEnumerator1"))
  (bind _returnvalue (Customization|GetSkeletalMeshforPartType (Variables|Default|GetFortCharacterParts) "Body" :!class /Script/FortniteGame.FortCharacterPartsComponent))
  (Rendering|Material|SetMaterial _returnvalue 0 (Utilities|Select :Index _selectedvariant :NewEnumerator0 "/SpringEvent/NugS/Skins/Materials/Fauna_Chicken_Slurp.Fauna_Chicken_Slurp" :NewEnumerator1 "/SpringEvent/NugS/Skins/Materials/Fauna_Chicken_Hop.Fauna_Chicken_Hop"))
  (Variables|Default|SetColorVariation (Utilities|Select :Index _selectedvariant :NewEnumerator0 0.31 :NewEnumerator1 0.41))
  (Ability|GameplayEffect|AddLooseGameplayTags self (GameplayTags|MakeGameplayTagContainerfromTag (Utilities|Select :Index _selectedvariant_2 :NewEnumerator0 "(TagName=\"SpringChicken.Variant1\")" :NewEnumerator1 "(TagName=\"SpringChicken.Variant2\")"))) ; Giving the chicken tags matching its variant\nin case this is needed for quests
  (Variables|SetMarkerDisplay_DisplayName (Utilities|Struct|MakeMarkedActorDisplayInfo _markerdisplay_displayname (Utilities|Select :Index _selectedvariant_2 :Option%200 "/SpringEvent/Icons/T-Icon-Fauna-Chicken-Slurp.T-Icon-Fauna-Chicken-Slurp" :Option%201 "/SpringEvent/Icons/T-Icon-Fauna-Chicken-Hop.T-Icon-Fauna-Chicken-Hop") _markerdisplay_customindicatorclass _markerdisplay_customdetailsclass _markerdisplay_primarycolor _markerdisplay_sound true _markerdisplay_screenclamping)))

(fn OnRep_ChosenVariant)

(fn OnRep_PhasedIn
  (if (Variables|Default|GetPhasedIn)
    (CallFunction|PhaseinNugS)))

(fn SetNugSCollisionEnabled
  (params (Enabled :type Boolean))
  (bind _returnvalue (GameplayTags|MakeGameplayTagContainerfromTag "(TagName=\"Gameplay.Status.KnockbackImmunity\")"))
  (bind _self self)
  (bind _weaponcapsulecollision (Variables|Default|GetWeaponCapsuleCollision))
  (Collision|SetCollisionProfileName (Variables|Character|GetCapsuleComponent) (select Enabled (Variables|Default|GetCapsuleCollisionProfile) "FortLandscapeAndWater") true :!class /Script/Engine.PrimitiveComponent)
  (Collision|SetCollisionResponsetoChannel _weaponcapsulecollision "ECC_GameTraceChannel7" (select Enabled "ECR_Block" "ECR_Ignore"))
  (Collision|SetCollisionResponsetoChannel _weaponcapsulecollision "ECC_GameTraceChannel1" (select Enabled "ECR_Overlap" "ECR_Ignore"))
  (Collision|SetCollisionProfileName (Variables|Character|GetMesh) (select Enabled (Variables|Default|GetMeshCollisionProfile) "FortLandscapeAndWater") true :!class /Script/Engine.PrimitiveComponent)
  (if Enabled
    (Ability|GameplayEffect|RemoveLooseGameplayTags _self _returnvalue true)
    (else
      (Ability|GameplayEffect|AddLooseGameplayTags _self _returnvalue true)))
  (Networking|SwitchHasAuthority
    (:Authority
      (AI|SetCanInteract Enabled)
      (Collision|SetCollisionResponsetoChannel (Variables|Character|GetCapsuleComponent) "ECC_Pawn" (select Enabled "ECR_Block" "ECR_Ignore")))))

(fn ConstructionScript
  (|CallParentFunction))
